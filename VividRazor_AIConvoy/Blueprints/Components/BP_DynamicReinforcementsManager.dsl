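; Server-side component for AI convoys and trains that runs Living World reinforcement encounters, optionally reserving AI density.
; Encounters are picked by capped playlist team size, with boss variants on trains depending on the boss spawn manager, and escalate on a looping timer.
; Spawned AI receive a shared threat, updated from their target changes with a cooldown, and raise boss events for music.
; Once paused, the encounter ends when no spawned AI remain alive.
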(event EventBeginPlay
  (bind _returnvalue (Components|GetOwner))
  ; Try to reserve AI Density
  (Utilities|SwitchHasAuthority
    (:Authority
      (Variables|InternalData|SetEncounterIsPaused false)
      (bind (_gamestate _playlist) (Class|FortAsyncActionCurrentPlaylistReady|PlaylistDataReadyAsync)
        (:OnReady
          (Variables|Default|SetTeamSize (Variables|/Script/FortniteGame.FortPlaylist|GetMaxTeamSize _playlist))
          (CallFunction|ValidateTeamSize)
          (Utilities|FlowControl|DelayUntilNextTick)
          (if (Variables|Default|GetShouldRegisterDensity)
            (bind _returnvalue_2 (Class|FortAthenaLivingWorldManager|RegisterDensityReservoir (Class|FortAthenaLivingWorldManager|GetLivingWorldManager) _returnvalue (Variables|Default|GetDensityCount) 12001.0))
            (if (not _returnvalue_2)
              (Fortnite|FortShippingLog self (Utilities|String|ToString(Text) (Utilities|Text|FormatText "NSLOCTEXT(\"[2D5C87AA04564239F650F78EBF3A3B94]\", \"9402C3EA46EE059D0B8FF6A5F0A9E967\", \"Trying to Reserve AI Density from {a}, but Return Value was returned as FALSE.\")" _returnvalue)) true))))))))

(event Custom|StartReinforcements
  (params (LivingWorldVolume :type Fort%20Athena%20Living%20World%20Volume%20Object%20Reference) (DriverPawn :type Fort%20Pawn%20Object%20Reference) (Target :type Actor%20Object%20Reference))
  (bind _targetpawn (Variables|InternalData|GetTargetPawn))
  (bind _driver_pawn (Variables|Default|GetDriver%20Pawn))
  ; Start the Encounter
  (Networking|SwitchHasAuthority
    (:Authority
      (if (Variables|Default|GetIsTrain)
        (bind _newtarget (CallFunction|GetTargetNearby))
        (Variables|InternalData|SetTargetPawn _newtarget)
        (else
          (Variables|InternalData|SetTargetPawn Target)
          (Variables|Default|SetDriver%20Pawn DriverPawn)))
      (bind (_encounter _returnvalue) (CallFunction|GetEncountertoSpawn))
      (if _returnvalue
        (bind (_isvalid _returnvalue_2) (Default|StartActivityEncounter (Game|GetGameState) LivingWorldVolume (CallFunction|GetEncounterReservoirOwner) _encounter))
        (if _isvalid
          (bind _output_get (Variables|InternalData|SetEncounterRef _returnvalue_2))
          (EventDispatchers|BindEventtoOnActorSpawned _output_get (EventDispatchers|CreateEvent :!event Custom|EncounterAISpawned) :!class /Script/LagerRuntime.FortAthenaLivingWorldEncounterInstance)
          (bind _returnvalue_3 (Utilities|Time|SetTimerbyEvent (EventDispatchers|CreateEvent :!event Custom|AdvanceEncounterStep) (ScalableFloat|GetValueatLevel (Variables|Default|GetActivityStepUpDelay)) true))
          (Variables|InternalData|SetHNDL_ActivityStepTimer _returnvalue_3)
          (Utilities|IsValid _targetpawn
            (:Is%20Valid
              (bind _asfort_pawn (Utilities|Casting|CastToFortPawn _targetpawn)
                (:then
                  (Variables|InternalData|SetCurrentThreat _asfort_pawn))
                (:CastFailed
                  _)))
            (:Is%20Not%20Valid
              _))
          (Utilities|IsValid _driver_pawn
            (:Is%20Valid
              (bind _targetactor (Class|BPFLAICommonFunctions|GetTargetActorfromBlackboard _driver_pawn))
              (Variables|InternalData|SetCurrentThreat _targetactor))))
        (else
          (Fortnite|FortShippingLog self (Utilities|String|ToString(Text) (Utilities|Text|FormatText "NSLOCTEXT(\"[2D5C87AA04564239F650F78EBF3A3B94]\", \"77E085B94D09CB407E1F369A8DA1686E\", \"Trying to spawn a Living World  Encounter from {a} for the specified playlist size {b}, but could not find one in the Map Structure! From Train? {c}\")" (Components|GetOwner self) (Variables|Default|GetTeamSize) (Variables|Default|GetIsTrain))) true))))))

(event Custom|EndReinforcements
  ; End the Reinforcements
  (Class|BPFLLagerCommunFunctions|EndActivityEncounter (Game|GetGameState) (Variables|InternalData|GetEncounterRef))
  (Class|FortAthenaLivingWorldManager|UnregisterDensityReservoir (Class|FortAthenaLivingWorldManager|GetLivingWorldManager) (CallFunction|GetEncounterReservoirOwner)))

(event Custom|PauseReinforcements
  (bind _encounterref (Variables|InternalData|GetEncounterRef))
  ; Pause the Encounter
  (Utilities|IsValid _encounterref
    (:Is%20Valid
      (Class|FortAthenaLivingWorldEncounterInstance|PauseEncounter _encounterref)
      (Variables|InternalData|SetEncounterIsPaused true)
      (Utilities|Time|ClearandInvalidateTimerbyHandle (Variables|InternalData|GetHNDL_ActivityStepTimer))
      (if (== (Class|FortAthenaLivingWorldEncounterInstance|GetAliveSpawnedActorCount _encounterref) 0)
        (CallFunction|EndReinforcements)))))

(event Custom|AdvanceEncounterStep
  ; Increase Encounter Stage
  (Class|FortAthenaLivingWorldEncounterInstance|AdvanceEncounterStage (Variables|InternalData|GetEncounterRef)))

(event Custom|EncounterAISpawned
  (params (EncounterInstance :type Fort%20Athena%20Living%20World%20Encounter%20Instance%20Object%20Reference) (SpawnedActor :type Actor%20Object%20Reference) (EventData :type Fort%20Athena%20Living%20World%20Event%20Data%20Object%20Reference))
  (bind _currentthreat (Variables|InternalData|GetCurrentThreat))
  ; New AI Encounter Spawned - Force Target
  (bind _asfort_pawn (Utilities|Casting|CastToFortPawn SpawnedActor))
  (bind _asfort_athena_aibot_controller (Utilities|Casting|CastToFortAthenaAIBotController (Pawn|GetController _asfort_pawn)))
  (Perception|BindEventtoOnCurrentTargetChangedEvent _asfort_athena_aibot_controller (EventDispatchers|CreateEvent :!event Custom|AITargetChanged))
  (Death|BindEventtoOnAgentDiedEvent _asfort_athena_aibot_controller (EventDispatchers|CreateEvent :!event Custom|SpawnAIDied))
  (Utilities|IsValid _currentthreat
    (:Is%20Valid
      (Class|FortAthenaAIBotController|AddTargetInfo _asfort_athena_aibot_controller _currentthreat false "Threat_Seeing" 1e+01 5e+03))
    (:Is%20Not%20Valid
      _))
  ; Music events
  (bind _asbp_bot_controller_boss_base (Utilities|Casting|CastToBP_BotController_Boss_Base _asfort_athena_aibot_controller))
  (Default|CallOnBossSpawned)
  (Death|BindEventtoOnAgentDiedEvent _asbp_bot_controller_boss_base (delegate Custom|Boss%20Died)))

(event Custom|Boss%20Died
  (params (BotController :type Fort%20Athena%20AIBot%20Controller%20Object%20Reference) (AgentPawn :type Fort%20Pawn%20Object%20Reference))
  ; Music events
  (Default|CallOnBossDied))

(event Custom|AITargetChanged
  (params (OldTarget :type Actor%20Object%20Reference) (NewTarget :type Actor%20Object%20Reference))
  ; Try to update Target if needed
  (if (or (> (Utilities|Time|GetGameTimeinSeconds) (+ (Variables|InternalData|GetLastTargetSwitched) (Variables|Default|GetTargetSwitchTime))) (not (Utilities|IsValid (Variables|InternalData|GetCurrentThreat) :!class /Script/Engine.KismetSystemLibrary)))
    (Variables|InternalData|SetLastTargetSwitched (Utilities|Time|GetGameTimeinSeconds))
    (bind _asfort_pawn (Utilities|Casting|CastToFortPawn NewTarget))
    (Variables|InternalData|SetCurrentThreat _asfort_pawn)))

(event Custom|SpawnAIDied
  (params (BotController :type Fort%20Athena%20AIBot%20Controller%20Object%20Reference) (AgentPawn :type Fort%20Pawn%20Object%20Reference))
  ; Validate when AI are dead to see if we should stop the Encounte ror not!
  (if (Variables|InternalData|GetEncounterIsPaused)
    (if (== (Class|FortAthenaLivingWorldEncounterInstance|GetAliveSpawnedActorCount (Variables|InternalData|GetEncounterRef)) 0)
      (CallFunction|EndReinforcements))))

(fn GetEncounterReservoirOwner
  (return (select (Utilities|IsValid (Variables|Default|GetCustomReservoirOwner) :!class /Script/Engine.KismetSystemLibrary) (Variables|Default|GetCustomReservoirOwner) (Components|GetOwner))))

(fn GetEncounterToSpawn
  (bind (_value _returnvalue) (Utilities|Map|Find (Variables|Default|GetEncounterAsset) (Variables|Default|GetTeamSize)))
  (bind _encounterassetwithboss (Variables|Default|GetEncounterAssetWithBoss))
  (bind (_value_2 _returnvalue_2) (Utilities|Map|Find _encounterassetwithboss (Variables|Default|GetTeamSize)))
  (bind (_value_3 _returnvalue_3) (Utilities|Map|Find (Variables|Default|GetEncounterAsset) (Variables|Default|GetTeamSize)))
  (bind _returnvalue_4 (Utilities|Map|Length _encounterassetwithboss))
  (bind _returnvalue_5 (> _returnvalue_4 0))
  (if (Variables|Default|GetIsTrain)
    (bind _outactors (Actor|GetAllActorsOfClass "/VividRazor_Minions/Blueprints/BP_VividRazor_BossSpawnLocation_Manager.BP_VividRazor_BossSpawnLocation_Manager_C"))
    (if (> (Utilities|Array|Length _outactors) 0)
      (bind _isbossspawningattrain (Class|BPVividRazorBossSpawnLocationManager|IsBossSupposedtoSpawnatTrain (Utilities|Array|Get(acopy) _outactors)))
      (if (not (and _isbossspawningattrain _returnvalue_5))
        (return _value _returnvalue))
      (else
        (Fortnite|FortShippingLog self "Trying to Spawn Encounter from the Train, but was not able to get the Boss Spawn Location Manager to know if the Boss should be spawning from us or not! Will try to allow Boss to spawn in case, better to try to have him than not!" true)
        (if (not _returnvalue_5)
          (return _value _returnvalue))))
    (return _value_2 _returnvalue_2)
    (else
      (return _value_3 _returnvalue_3))))

(fn GetTargetNearby
  (bind (_outhits _returnvalue) (Collision|MultiSphereTraceForObjects (Transformation|GetActorLocation (Components|GetOwner)) (+ (Transformation|GetActorLocation (Components|GetOwner)) "0.000000,0.000000,100.000000") (Variables|Default|GetRadius) (Variables|Default|GetObjectTypesForGettingNearbyTargets)))
  (if _returnvalue
    (bind _array_element (Utilities|Array|ForEachLoopwithBreak _outhits :!label _label)
      (:LoopBody
        (bind _asfort_player_pawn_athena (Utilities|Casting|CastToFortPlayerPawnAthena (.HitActor (Collision|BreakHitResult _array_element))))
        (if (not (GameplayTags|HasAnyMatchingGameplayTags _asfort_player_pawn_athena (Variables|Default|GetTagContainerForNPC)))
          (LocalVariables|SetFoundPlayerOrBot _asfort_player_pawn_athena)
          (goto _label Break)))
      (:Completed
        (return (LocalVariables|GetFoundPlayerOrBot))))))

(fn SetEncounterReservoirOwner
  (params (ReservoirOwner :type Actor%20Object%20Reference))
  (Variables|Default|SetCustomReservoirOwner ReservoirOwner))

(fn ValidateTeamSize
  (Variables|Default|SetTeamSize (Math|Integer|Min(Integer) (Math|Integer|Max(Integer) (Variables|Default|GetTeamSize)) 4))
  (return))
